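# UI material for a translucent badge in the crew unlocking screen, with a shine effect driven by a Shine Amount parameter.
# A parameterised badge texture is sampled on scaled-to-fit UVs and split into color and alpha.
# MF_Shine computes the rotated highlight on UVs distorted by BumpOffset_Advanced, which uses a desaturated input as height and a parameter as camera vector.
# A divide by a clamped one-minus term forms a dodge-style brightening.
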Material /Crew/UI/Materials/CrewUnlocking/M_UI_Crew_Unlocking_Badge
graph {
 properties {
  bAutomaticallySetUsageInEditor=false
  bUsedWithStaticMesh=false
  blendMode=BLEND_Translucent
  materialDomain=MD_UI
  usedShadingModels=Default Lit
 }
 nodes [
  n1 = NamedRerouteUsage()
  Divide(A=NamedRerouteUsage(), B=OneMinus(None=Min(A=@n1) {constB=0.8}))
  n2 = TextureSampleParameter2D(UVs=NamedRerouteDeclaration(None=MaterialFunctionCall() {materialFunction=&MF_UI_UVScaling}.Scaled to Fit) {name=UV, nodeColor=<rgba 0 0 0 1>}) {parameterName=BadgeTexture, texture=/Crew/UI/Textures/T_UI_Nexus_BattlePass}
  NamedRerouteDeclaration(None=@n2.RGB) {name=Badge Texture, nodeColor=<rgba 0 0 0 1>}
  NamedRerouteDeclaration(None=@n2.A) {name=Badge Alpha, nodeColor=<rgba 0 0 0 1>}
  n3 = VectorParameter() {parameterName=ShineExtents, defaultValue=<rgba -1 1 0 1>}
  n4 = ScalarParameter() {parameterName=Shine Amount}
  n5 = Desaturation(None=NamedRerouteUsage())
  n6 = NamedRerouteUsage()
  n7 = ScalarParameter() {defaultValue=2, parameterName="HeightRatioInput (S)"}
  n8 = VectorParameter() {parameterName="CameraVector (V3)", defaultValue=<rgba 0.2 0 0 0>}
  NamedRerouteDeclaration(None=MaterialFunctionCall!(
    ShineOffset=LinearInterpolate(A=@n3.R, B=@n3.R, Alpha=@n4) {constA=1, constB=-1}
    Rotation=ScalarParameter() {defaultValue=-0.7962, parameterName="Rotation (S)"}
    RotationCenter=VectorParameter() {parameterName="RotationCenter (V2)", defaultValue=<rgba 0.5 0.5 0 0>}.RGB
    UVs=MaterialFunctionCall!(HeightTexture=@n5, TextureCoordinates=@n6, ReferencePlane, HeightRatioInput=@n7, CameraVector=@n8.RGB) {materialFunction=/Engine/Functions/Engine_MaterialFunctions02/Utility/BumpOffset_Advanced.BumpOffset_advanced}.UVs
   ) {materialFunction=/Game/Athena/HUD/Materials/Victory/MF_Shine}.Result) {name=Shine, nodeColor=<rgba 0 0 0 1>}
  NamedRerouteUsage()
 ]
}
